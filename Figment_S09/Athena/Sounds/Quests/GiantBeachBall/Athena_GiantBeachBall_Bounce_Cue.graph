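# Sound cue for the bounce of the giant beach ball in Athena quest content.
# A Random node picks one of three bounce waves with equal weight and passes it through a Modulator.
# The cue plays in the gadgets sound class at reduced volume and uses the shared Athena ball-bounce attenuation asset for distance falloff.
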SoundCue /Figment_S09/Athena/Sounds/Quests/GiantBeachBall/Athena_GiantBeachBall_Bounce_Cue
root=Modulator_0
nodes [
 {id=Modulator_0, children [Random_0]}
 {id=Random_0, children [WavePlayer_1,WavePlayer_2,WavePlayer_3], properties {weights [1,1,1]}}
 {id=WavePlayer_1, properties {soundWaveAssetPtr=&/Athena_GiantBeachBall_Bounce_02}}
 {id=WavePlayer_2, properties {soundWaveAssetPtr=&/Athena_GiantBeachBall_Bounce_03}}
 {id=WavePlayer_3, properties {soundWaveAssetPtr=&/Athena_GiantBeachBall_Bounce_01}}
]
properties {
 attenuationSettings=/Game/Sounds/Attenuation/Fort_Ball_Bounce_Athena_Att
 maxDistance=7000
 soundClassObject=&Fort_Gadgets
 volumeMultiplier=0.4
}
