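# Sound cue for the giant beach ball quest prop's impact sound.
# A mixer layers two randomized branches: one of three beach ball bounce waves pitched up, and one of three reused big rubber duck impact waves pitched slightly down at half volume.
# Each branch picks with equal weight and varies pitch through a modulator.
# It plays in the gadgets sound class with a ball bounce attenuation asset.
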SoundCue /Figment_S09/Athena/Sounds/Quests/GiantBeachBall/Athena_GiantBeachBall_Impact_Cue
root=Mixer_0
nodes [
 {id=Mixer_0, children [Modulator_0,Modulator_1], properties {inputVolume [1,0.5]}}
 {id=Modulator_0, children [Random_0], properties {pitchMax=1.4, pitchMin=1.2}}
 {id=Modulator_1, children [Random_1], properties {pitchMax=0.95, pitchMin=0.9}}
 {id=Random_0, children [WavePlayer_1,WavePlayer_2,WavePlayer_3], properties {weights [1,1,1]}}
 {id=Random_1, children [WavePlayer_0,WavePlayer_4,WavePlayer_5], properties {weights [1,1,1]}}
 {id=WavePlayer_0, properties {soundWaveAssetPtr=/FNGameplayCues/Athena/Sounds/Quests/BigRubberDuck_Impacts_03}}
 {id=WavePlayer_1, properties {soundWaveAssetPtr=&/Athena_GiantBeachBall_Bounce_02}}
 {id=WavePlayer_2, properties {soundWaveAssetPtr=&/Athena_GiantBeachBall_Bounce_03}}
 {id=WavePlayer_3, properties {soundWaveAssetPtr=&/Athena_GiantBeachBall_Bounce_01}}
 {id=WavePlayer_4, properties {soundWaveAssetPtr=/FNGameplayCues/Athena/Sounds/Quests/BigRubberDuck_Impacts_04}}
 {id=WavePlayer_5, properties {soundWaveAssetPtr=/FNGameplayCues/Athena/Sounds/Quests/BigRubberDuck_Impacts_05}}
]
properties {
 attenuationSettings=/Game/Sounds/Attenuation/Fort_Ball_Bounce_Athena_Att
 maxDistance=7000
 soundClassObject=&Fort_Gadgets
}
